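; Tree sapling quest actor for the Season 22 narrative quests that animates a growing tree after a player interacts with it.
; Its PostInteractionVisuals override calls the parent, then triggers ScaleUpTree only on the remote (non-authority) side, keeping the animation cosmetic.
; ScaleUpTree plays a timeline from the start, driving uniform relative scale on the SM_Quest_PostInteraction mesh component.
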(event EventBeginPlay
  (|CallParentFunction) ; This node is disabled and will not be called.\nDrag off pins to build functionality.
)

(event Collision|EventActorBeginOverlap
  (params (OtherActor :type Actor%20Object%20Reference))
  (|CallParentFunction OtherActor) ; This node is disabled and will not be called.\nDrag off pins to build functionality.
)

(event EventTick
  (params (DeltaSeconds :type Float%20%28single-precision%29))
  (|CallParentFunction DeltaSeconds) ; This node is disabled and will not be called.\nDrag off pins to build functionality.
)

(event Custom|ScaleUpTree
  (bind (_direction _scaleuptrack) (Timeline|ScaleUpTimeline PlayFromStart)
    (:Update
      (Transformation|SetRelativeScale3D (Variables|Components|GetSM_Quest_PostInteraction) (Math|Conversions|ToVector(Float) _scaleuptrack)))))

(fn PostInteractionVisuals
  (params (PlayerController :type Fort%20Player%20Controller%20Object%20Reference))
  (|CallParentFunction PlayerController)
  (Networking|SwitchHasAuthority
    (:Remote
      (CallFunction|ScaleUpTree))))

(fn ConstructionScript
  (|CallParentFunction))
